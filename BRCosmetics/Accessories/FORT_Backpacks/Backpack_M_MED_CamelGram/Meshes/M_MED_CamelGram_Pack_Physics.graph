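# Physics asset for the CamelGram backpack cosmetic, defining capsule bodies and constraints for its dangling legs, necklace, neck and head.
# Left leg, right leg and necklace each form a three-link chain from a small base body, which joins the pack body with free rotation.
# Chain links have twist locked and modest per-joint swing limits.
# Neck and head use tight angular limits, with an extra free-rotation constraint tying head directly to pack.
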PhysicsAsset /BRCosmetics/Accessories/FORT_Backpacks/Backpack_M_MED_CamelGram/Meshes/M_MED_CamelGram_Pack_Physics
bodies {
 base_leg_l [{shapeType=Capsule, length=0.505, radius=0.505, shapeName=base_leg_l_capsule, rotation=<rot 0 -90 0>}]
 base_leg_r [{shapeType=Capsule, length=0.505, radius=0.505, shapeName=base_leg_r_capsule, rotation=<rot 0 -90 0>}]
 base_necklace [{shapeType=Capsule, length=0, radius=0.505, shapeName=base_necklace_capsule}]
 dyn_head [{shapeType=Capsule, length=0, radius=3, shapeName=dyn_head_capsule, center=<2 0 0>}]
 dyn_leg_l_01 [{shapeType=Capsule, length=0.9406, radius=2.1542, shapeName=dyn_leg_l_01_capsule, center=<2 0 0>, rotation=<rot 90 0 0>}]
 dyn_leg_l_02 [{shapeType=Capsule, length=0.9949, radius=1.9627, shapeName=dyn_leg_l_02_capsule, center=<2 0 0>, rotation=<rot 90 0 0>}]
 dyn_leg_l_03 [{shapeType=Capsule, length=1.1518, radius=2.7757, shapeName=dyn_leg_l_03_capsule, center=<2 0 1.1656>, rotation=<rot 0 0 90>}]
 dyn_leg_r_01 [{shapeType=Capsule, length=0, radius=3.1572, shapeName=dyn_leg_r_01_capsule, center=<2 0 0>, rotation=<rot 90 0 0>}]
 dyn_leg_r_02 [{shapeType=Capsule, length=0, radius=2.9838, shapeName=dyn_leg_r_02_capsule, center=<2 0 0>, rotation=<rot 90 0 0>}]
 dyn_leg_r_03 [{shapeType=Capsule, length=1.1518, radius=2.7757, shapeName=dyn_leg_r_03_capsule, center=<2 0 1.1656>, rotation=<rot 0 0 90>}]
 dyn_neck [{shapeType=Capsule, length=0, radius=3, shapeName=dyn_neck_capsule, center=<2 0 0>}]
 dyn_necklace_01 [{shapeType=Capsule, length=6.9783, radius=0.7815, shapeName=dyn_necklace_01_capsule, center=<2 0 0.3367>, rotation=<rot 0 0 90>}]
 dyn_necklace_02 [{shapeType=Capsule, length=4.4343, radius=0.8894, shapeName=dyn_necklace_02_capsule, center=<1.25 0 0.3367>, rotation=<rot 0 0 90>}]
 dyn_necklace_03 [{shapeType=Capsule, length=2.4618, radius=0.9444, shapeName=dyn_necklace_03_capsule, center=<2.2731 0 0.6495>, rotation=<rot 90 0 0>}]
 pack [{shapeType=Capsule, length=0.1606, radius=5.7398, center=<0 -6.9225 -0.1199>}]
}
constraints [
 {
  bone1Name=base_necklace
  bone2Name=pack
  swing1LimitDegrees=45
  swing1Motion=Free
  swing2LimitDegrees=45
  swing2Motion=Free
  twistLimitDegrees=45
  twistMotion=Free
 }
 {bone1Name=dyn_necklace_01, bone2Name=base_necklace, swing1LimitDegrees=5, swing2LimitDegrees=10, twistLimitDegrees=0, twistMotion=Locked}
 {bone1Name=dyn_necklace_02, bone2Name=dyn_necklace_01, swing1LimitDegrees=10, swing2LimitDegrees=15, twistLimitDegrees=0, twistMotion=Locked}
 {bone1Name=dyn_necklace_03, bone2Name=dyn_necklace_02, swing1LimitDegrees=15, swing2LimitDegrees=25, twistLimitDegrees=0, twistMotion=Locked}
 {
  bone1Name=base_leg_l
  bone2Name=pack
  swing1LimitDegrees=45
  swing1Motion=Free
  swing2LimitDegrees=45
  swing2Motion=Free
  twistLimitDegrees=45
  twistMotion=Free
 }
 {bone1Name=dyn_leg_l_01, bone2Name=base_leg_l, swing1LimitDegrees=25, swing2LimitDegrees=10, twistLimitDegrees=0, twistMotion=Locked}
 {bone1Name=dyn_leg_l_02, bone2Name=dyn_leg_l_01, swing1LimitDegrees=25, swing2LimitDegrees=20, twistLimitDegrees=0, twistMotion=Locked}
 {bone1Name=dyn_leg_l_03, bone2Name=dyn_leg_l_02, swing1LimitDegrees=16, swing2LimitDegrees=8, twistLimitDegrees=0, twistMotion=Locked}
 {
  bone1Name=base_leg_r
  bone2Name=pack
  swing1LimitDegrees=45
  swing1Motion=Free
  swing2LimitDegrees=45
  swing2Motion=Free
  twistLimitDegrees=45
  twistMotion=Free
 }
 {bone1Name=dyn_leg_r_01, bone2Name=base_leg_r, swing1LimitDegrees=20, swing2LimitDegrees=8, twistLimitDegrees=0, twistMotion=Locked}
 {bone1Name=dyn_leg_r_02, bone2Name=dyn_leg_r_01, swing1LimitDegrees=20, swing2LimitDegrees=16, twistLimitDegrees=0, twistMotion=Locked}
 {bone1Name=dyn_leg_r_03, bone2Name=dyn_leg_r_02, swing1LimitDegrees=12, swing2LimitDegrees=5, twistLimitDegrees=0, twistMotion=Locked}
 {bone1Name=dyn_head, bone2Name=dyn_neck, swing1LimitDegrees=6, swing2LimitDegrees=6, twistLimitDegrees=6}
 {bone1Name=dyn_neck, bone2Name=pack, swing1LimitDegrees=6, swing2LimitDegrees=6, twistLimitDegrees=0}
 {
  bone1Name=dyn_head
  bone2Name=pack
  swing1LimitDegrees=45
  swing1Motion=Free
  swing2LimitDegrees=45
  swing2Motion=Free
  twistLimitDegrees=45
  twistMotion=Free
 }
]
